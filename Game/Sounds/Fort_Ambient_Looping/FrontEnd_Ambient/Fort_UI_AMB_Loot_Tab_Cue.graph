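# Sound cue for the looping ambient bed of the front-end Loot tab in the UI.
# A single wave player loops the AMB_Loot_Tab_01_Loop wave straight to the output at reduced volume.
# It uses the Fort_UI sound class, and a concurrency override allows one instance, blocking new plays while it is active.
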SoundCue /Game/Sounds/Fort_Ambient_Looping/FrontEnd_Ambient/Fort_UI_AMB_Loot_Tab_Cue
root=WavePlayer_1
nodes [{id=WavePlayer_1, properties {bLooping=true, soundWaveAssetPtr=&/AMB_Loot_Tab_01_Loop}}]
properties {
 maxDistance=2097152
 soundClassObject=&Fort_UI
 volumeMultiplier=0.4
 concurrencyOverrides {maxCount=1, resolutionRule=PreventNew}
}
